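; Parent Blueprint for Hermes chests that links each chest to the environment beneath it so it does not float when that piece is destroyed.
; On world ready, the server runs a downward linking check, retrying on a timer up to a maximum to cover late-loading environment, then logging failure.
; Destruction of the linked actor fades out the ambient sound and destroys the chest shortly after.
; LOSCheck is a placeholder that always returns true.
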(event EventOnWorldReady
  (|CallParentFunction)
  (Networking|SwitchHasAuthority
    (:Authority
      (CallFunction|TrytoLinktoEnvironment))))

(event EventBeginPlay
  (|CallParentFunction))

(event Collision|EventActorBeginOverlap
  (params (OtherActor :type Actor%20Object%20Reference))
  (|CallParentFunction OtherActor) ; This node is disabled and will not be called.\nDrag off pins to build functionality.
)

(event EventTick
  (params (DeltaSeconds :type Float%20%28single-precision%29))
  (|CallParentFunction DeltaSeconds) ; This node is disabled and will not be called.\nDrag off pins to build functionality.
)

(event OnLinkedActorDestroyed(FortLinkToActor)
  (params (DamageCauser :type Actor%20Object%20Reference))
  ;[0.4149 0.164 0] OnLinkedActorDestroyed
  ; Set the life span to destroy the alarm when the attached envionment actor is destoyed
  (bind _chest_ambient_sound (Variables|Default|GetChest_Ambient_Sound)
    (:then
      (Audio|Components|Audio|FadeOut _chest_ambient_sound 0.1 0.0 "Linear" :!class /Script/Engine.AudioComponent)
      (Components|Activation|Deactivate _chest_ambient_sound)
      (Actor|SetLifeSpan 0.2))
    (:else)))

(event EventOnLoot
  (|CallParentFunction))

(event EventEndPlay
  (params (EndPlayReason :type EEndPlayReason%20Enum :default "Destroyed"))
  (Utilities|Time|ClearandInvalidateTimerbyHandle (Variables|LinktoEnvironment|GetTimerH_RetryLinkToEnvironment))
  (|CallParentFunction EndPlayReason))

(event Custom|TryToLinkToEnvironment
  ;[0.3798 0.6987 1] TryToLinkToEnvironment
  ; Attempt to attach the chest to the environment to prevent it from floating if the env is destroyed
  (Math|Integer|IncrementInt (Variables|LinktoEnvironment|GetLinkAttemptsCount)
    (:%20%20
      (bind _returnvalue (LinkToActorComponent|PerformLinkingCheck (Variables|Default|GetFortLinkToActor) "Down" (Variables|LinkToEnvironment|GetEnvironmentLinkedActor)))
      ;;[1 0 0.8947] Fallback handles case where the chest which is in a content bundle in Asteria is loaded before the environment piece it is trying to attach to (in a level instance). Linking to an indestructible actor (ex terrain) underneath will fail
      (if _returnvalue
        _
        (elif (>= (Variables|LinktoEnvironment|GetLinkAttemptsCount) (Variables|LinktoEnvironment|GetLinkAttemptsMax))
          (Fortnite|FortShippingLog self (Utilities|String|Append "HERMES CHEST: Attempt to Link to Environment Failed on total attempts number:  " (Utilities|String|ToString(Integer) (Variables|LinktoEnvironment|GetLinkAttemptsMax))))
          (else
            (bind _returnvalue_2 (Utilities|Time|SetTimerbyEvent (EventDispatchers|CreateEvent :!event Custom|TryToLinkToEnvironment) (Variables|LinktoEnvironment|GetLinkAttemptTiming))) ; try to link again
            (Variables|LinktoEnvironment|SetTimerH_RetryLinkToEnvironment _returnvalue_2)))))))

(fn LOSCheck
  (params (TracedActor :type Actor%20Object%20Reference))
  ; TODO: Set this up later
  (return true))

(fn ConstructionScript
  (|CallParentFunction))
